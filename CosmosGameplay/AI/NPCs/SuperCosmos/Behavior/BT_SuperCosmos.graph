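# Behavior tree for the SuperCosmos NPC, covering air movement, survival, combat, investigation and patrol on the shared NPC blackboard.
# A priority selector tries diving and gliding, then ground overrides such as launch steering, threat avoidance, downed movement and leash return.
# Alert level then selects looping Threatened, last-known-position, Alerted or Unaware branches, with revive, converted-follower, conversation and patrol behaviors.
# Evaluator services write the status and destination keys that gate branches, with Phoebe subtrees reused.
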BehaviorTree /CosmosGameplay/AI/NPCs/SuperCosmos/Behavior/BT_SuperCosmos
blackboard=/Game/Athena/AI/NPCs/Base/BehaviorTree/BB_NPC
root {
 name=Alive
 class=BTComposite_Selector
 children [
  Diving {
   class=FortAthenaBTTask_Dive
   children []
   decorators [Can Dive {
     class=BTDecorator_Blackboard
     properties {
      arithmeticOperation=Equal
      intValue=4
      nodeName=Can Dive
      stringValue=EExecutionStatus::ExecutionAllowed
      blackboardKey {selectedKeyName=AIEvaluator_Dive_ExecutionStatus}
     }
    }]
   properties {
    bIgnoreRestartSelf=false
    diveDestinationKeyName=AIEvaluator_Dive_Destination
    executionStatusKeyName=AIEvaluator_Dive_ExecutionStatus
    nodeName=Diving
   }
   services []
  }
  Gliding {
   class=FortAthenaBTTask_Glide
   children []
   decorators [Can Glide {
     class=BTDecorator_Blackboard
     properties {
      arithmeticOperation=Equal
      intValue=4
      nodeName=Can Glide
      stringValue=EExecutionStatus::ExecutionAllowed
      blackboardKey {selectedKeyName=AIEvaluator_Glide_ExecutionStatus}
     }
    }]
   properties {
    bIgnoreRestartSelf=false
    executionStatusKeyName=AIEvaluator_Glide_ExecutionStatus
    glideDestinationKeyName=AIEvaluator_Glide_Destination
    nodeName=Gliding
   }
   services []
  }
  On Ground {
   class=BTComposite_Selector
   children [
    Wait {
     class=BTTask_Wait
     decorators [
      Is EndGame? {
       class=BTDecorator_Blackboard
       properties {
        arithmeticOperation=GreaterOrEqual
        intValue=11
        nodeName=Is EndGame?
        stringValue=EAthenaGamePhaseStep::EndGame
        blackboardKey {selectedKeyName=AIEvaluator_Global_GamePhaseStep}
       }
      }
      Loop {class=BTDecorator_Loop}
     ]
     properties {waitTime {defaultValue=5}}
    }
    Wait {
     class=BTTask_Wait
     decorators [
      Is Movement Blocked? {
       class=BTDecorator_Blackboard
       properties {
        arithmeticOperation=Equal
        intValue=0
        nodeName=Is Movement Blocked?
        stringValue=""
        blackboardKey {selectedKeyName=AIEvaluator_Global_IsMovementBlocked}
       }
      }
      Loop {class=BTDecorator_Loop}
     ]
     properties {waitTime {defaultValue=1}}
    }
    Steer Movement {
     class=FortAthenaBTTask_SteerMovement
     children []
     decorators [Is Character being launched? {
       class=BTDecorator_Blackboard
       properties {
        arithmeticOperation=GreaterOrEqual
        intValue=3
        nodeName=Is Character being launched?
        stringValue=EExecutionStatus::ExecutionPending
        blackboardKey {selectedKeyName=AIEvaluator_CharacterLaunched_ExecutionStatus}
       }
      }]
     properties {
      bIgnoreRestartSelf=false
      bSetControlRotation=false
      nodeName=Steer Movement
      steerDirectionKeySelector {bNoneIsAllowedValue=false, selectedKeyName=AIEvaluator_CharacterLaunched_SteerDirection}
     }
     services []
    }
    Bot Shoots Trap {
     class=FortAthenaBTTask_ShootTrap
     children []
     decorators [Trap on Path? {
       class=BTDecorator_Blackboard
       properties {
        arithmeticOperation=GreaterOrEqual
        intValue=3
        nodeName=Trap on Path?
        stringValue=EExecutionStatus::ExecutionPending
        blackboardKey {selectedKeyName=AIEvaluator_TrapOnPath_ExecutionStatus}
       }
      }]
     properties {
      bIgnoreRestartSelf=false
      nodeName=Bot Shoots Trap
      trapOnPathKeyName=AIEvaluator_TrapOnPath_ExecutionStatus
      randomDeviation {defaultValue=0, key=None}
      targetActorKey {bNoneIsAllowedValue=false, selectedKeyName=AIEvaluator_TrapOnPath_Actor}
      waitTime {defaultValue=5, key=None}
     }
     services []
    }
    AvoidThreat MoveTo {
     class=FortAthenaBTTask_BotMoveTo
     decorators [Can AvoidThreat? {
       class=BTDecorator_Blackboard
       properties {
        arithmeticOperation=GreaterOrEqual
        intValue=4
        nodeName=Can AvoidThreat?
        stringValue=EExecutionStatus::ExecutionAllowed
        blackboardKey {selectedKeyName=AIEvaluator_AvoidThreat_ExecutionStatus}
       }
      }]
     properties {
      movementResultKeyName=AIEvaluator_AvoidThreat_MovementState
      nodeName=AvoidThreat MoveTo
      acceptableRadius {defaultValue=100}
      allowRandomWobble {defaultValue=false}
      bAllowStrafe {defaultValue=true}
      blackboardKey {selectedKeyName=AIEvaluator_AvoidThreat_Destination}
      observedBlackboardValueTolerance {defaultValue=4.75}
     }
    }
    Move away from danger {
     class=FortAthenaBTTask_BotMoveTo
     decorators [Danger detected? {
       class=BTDecorator_Blackboard
       properties {
        arithmeticOperation=GreaterOrEqual
        intValue=3
        nodeName=Danger detected?
        stringValue=EExecutionStatus::ExecutionPending
        blackboardKey {selectedKeyName=AIEvaluator_DangerZoneDetected_ExecutionStatus}
       }
      }]
     properties {
      movementResultKeyName=None
      nodeName=Move away from danger
      acceptableRadius {defaultValue=5}
      allowRandomWobble {defaultValue=false}
      bAllowStrafe {defaultValue=true}
      bTrackMovingGoal {defaultValue=false}
      blackboardKey {selectedKeyName=AIEvaluator_DangerZoneDetected_SafeLocation}
      observedBlackboardValueTolerance {defaultValue=4.75}
     }
    }
    Selector {
     class=BTComposite_Selector
     children [
      Bot Move To {
       class=FortAthenaBTTask_BotMoveTo
       decorators [Can Do DBNO Move {
         class=BTDecorator_Blackboard
         properties {
          arithmeticOperation=GreaterOrEqual
          intValue=3
          nodeName=Can Do DBNO Move
          stringValue=EExecutionStatus::ExecutionPending
          blackboardKey {selectedKeyName=AIEvaluator_DBNO_ExecutionStatus}
         }
        }]
       properties {
        movementResultKeyName=AIEvaluator_DBNO_MovementState
        nodeName=Bot Move To
        acceptableRadius {defaultValue=150}
        allowRandomWobble {defaultValue=false}
        bAllowStrafe {defaultValue=false}
        blackboardKey {selectedKeyName=AIEvaluator_DBNO_Destination}
        filterClass {defaultValue=&NavFilter_Phoebe_NoSmashable_C}
        observedBlackboardValueTolerance {defaultValue=4.75}
       }
      }
      Wait {class=BTTask_Wait, properties {waitTime {defaultValue=1}}}
     ]
     decorators [
      Is DBNO {
       class=BTDecorator_Blackboard
       properties {arithmeticOperation=GreaterOrEqual, intValue=3, nodeName=Is DBNO, stringValue="", blackboardKey {selectedKeyName=AIEvaluator_Global_DBNO}}
      }
      Loop {class=BTDecorator_Loop}
     ]
     properties {
      children [
       {
        childComposite=null
        childTask=&self:FortAthenaBTTask_BotMoveTo_2
        decorators [&self:BTDecorator_Blackboard_10]
       }
       {childComposite=null, childTask=&self:BTTask_Wait_4, decorators []}
      ]
      services [&self:FortAthenaBTService_AIEvaluator_22]
     }
     services [Evaluating...DBNO Move {
       class=FortAthenaBTService_AIEvaluator
       properties {
        bCallTickOnSearchStart=true
        interval=0.1
        nodeName=Evaluating...DBNO Move
        randomDeviation=0
        aIEvaluatorClass {defaultValue=/Game/Athena/AI/NPCs/Base/Evaluators/Evaluator_DBNO_Guard.Evaluator_DBNO_Guard_C}
       }
      }]
    }
    Ziplining {
     class=FortAthenaBTTask_Zipline
     children []
     decorators [Is on Zipline {
       class=BTDecorator_Blackboard
       properties {
        arithmeticOperation=GreaterOrEqual
        intValue=4
        nodeName=Is on Zipline
        stringValue=EExecutionStatus::ExecutionAllowed
        blackboardKey {selectedKeyName=AIEvaluator_Zipline_UsageExecutionStatus}
       }
      }]
     properties {
      bIgnoreRestartSelf=false
      nodeName=Ziplining
      usageExecutionStatusName=AIEvaluator_Zipline_UsageExecutionStatus
      ziplineTargetName=AIEvaluator_Zipline_Target
     }
     services []
    }
    Selector {
     class=BTComposite_Selector
     children [
      Run Unstuck Behavior {
       class=BTTask_RunBehavior
       decorators []
       properties {behaviorAsset=&BT_Phoebe_UnstuckNew, nodeName=Run Unstuck Behavior}
      }
      Back into Leash MoveTo {
       class=FortAthenaBTTask_BotMoveTo
       decorators [Is Converted? {
         class=BTDecorator_CheckGameplayTagsOnActor
         properties {
          bInverseCondition=false
          flowAbortMode=None
          nodeName=Is Converted?
          tagsToMatch=Any
          actorToCheck {bNoneIsAllowedValue=false, selectedKeyName=SelfActor}
          gameplayTags {gameplayTags [{tagName=Athena.Gameplay.Status.Converted}]}
         }
        }]
       properties {
        movementResultKeyName=AIEvaluator_Leash_MovementState
        nodeName=Back into Leash MoveTo
        acceptableRadius {defaultValue=10}
        allowRandomWobble {defaultValue=false}
        bAllowStrafe {defaultValue=true}
        blackboardKey {selectedKeyName=AIEvaluator_Leash_Destination}
        observedBlackboardValueTolerance {defaultValue=4.75}
       }
      }
      Back into Leash MoveTo {
       class=FortAthenaBTTask_BotMoveTo
       decorators []
       properties {
        movementResultKeyName=AIEvaluator_Leash_MovementState
        nodeName=Back into Leash MoveTo
        acceptableRadius {defaultValue=10}
        allowRandomWobble {defaultValue=false}
        bAllowStrafe {defaultValue=true}
        blackboardKey {selectedKeyName=AIEvaluator_Leash_Destination}
        observedBlackboardValueTolerance {defaultValue=4.75}
       }
      }
     ]
     decorators [Must Move back into Leash? {
       class=BTDecorator_Blackboard
       properties {
        arithmeticOperation=GreaterOrEqual
        intValue=4
        nodeName=Must Move back into Leash?
        stringValue=EExecutionStatus::ExecutionAllowed
        blackboardKey {selectedKeyName=AIEvaluator_Leash_ExecutionStatus}
       }
      }]
     properties {
      children [
       {childComposite=null, childTask=&self:BTTask_RunBehavior_1, decorators []}
       {
        childComposite=null
        childTask=&self:FortAthenaBTTask_BotMoveTo_0
        decorators [&self:BTDecorator_CheckGameplayTagsOnActor_2]
       }
       {childComposite=null, childTask=&self:FortAthenaBTTask_BotMoveTo_6, decorators []}
      ]
      services []
     }
     services []
    }
    THREATEN {
     class=BTComposite_Selector
     children [Loop Threaten {
       class=BTComposite_Selector
       children [
        Move To Range {
         class=BTTask_RunBehavior
         decorators [Can do Range {
           class=BTDecorator_Blackboard
           properties {
            arithmeticOperation=GreaterOrEqual
            intValue=3
            nodeName=Can do Range
            stringValue=EExecutionStatus::ExecutionPending
            blackboardKey {selectedKeyName=AIEvaluator_RangeAttack_ExecutionStatus}
           }
          }]
         properties {behaviorAsset=&BT_Phoebe_MoveToRange, nodeName=Move To Range}
        }
        Move To Melee {
         class=FortAthenaBTTask_BotMoveTo
         decorators [
          Can Do MeleeAttack {
           class=BTDecorator_Blackboard
           properties {
            arithmeticOperation=GreaterOrEqual
            intValue=3
            nodeName=Can Do MeleeAttack
            stringValue=EExecutionStatus::ExecutionPending
            blackboardKey {selectedKeyName=AIEvaluator_MeleeAttack_ExecutionStatus}
           }
          }
          Loop {class=BTDecorator_Loop}
         ]
         properties {
          movementResultKeyName=AIEvaluator_MeleeAttack_MovementState
          nodeName=Move To Melee
          acceptableRadius {defaultValue=5}
          allowRandomWobble {defaultValue=false}
          bAllowPartialPath {defaultValue=false}
          bAllowStrafe {defaultValue=false}
          blackboardKey {selectedKeyName=AIEvaluator_Attack_Destination}
          observedBlackboardValueTolerance {defaultValue=4}
         }
        }
        Fallback {class=BTTask_Wait, properties {nodeName=Fallback, waitTime {defaultValue=5}}}
       ]
       decorators [Loop {class=BTDecorator_Loop}]
       properties {
        nodeName=Loop Threaten
        children [
         {
          childComposite=null
          childTask=&self:BTTask_RunBehavior_10
          decorators [&self:BTDecorator_Blackboard_35]
         }
         {
          childComposite=null
          childTask=&self:FortAthenaBTTask_BotMoveTo_8
          decorators [
           &self:BTDecorator_Blackboard_19
           &self:BTDecorator_Loop_7
          ]
         }
         {childComposite=null, childTask=&self:BTTask_Wait_1, decorators []}
        ]
        services []
       }
       services []
      }]
     decorators [IsThreatened {
       class=BTDecorator_Blackboard
       properties {
        arithmeticOperation=Equal
        intValue=3
        nodeName=IsThreatened
        stringValue=EAlertLevel::Threatened
        blackboardKey {selectedKeyName=AIEvaluator_Global_AlertLevel}
       }
      }]
     properties {
      nodeName=THREATEN
      children [{
        childComposite=&self:BTComposite_Selector_11
        childTask=null
        decorators [&self:BTDecorator_Loop_0]
       }]
      services [
       &self:FortAthenaBTService_AIEvaluator_39
       &self:FortAthenaBTService_AIEvaluator_5
       &self:FortAthenaBTService_AIEvaluator_0
       &self:FortAthenaBTService_ManageWeapon_0
       &self:FortAthenaBTService_AIEvaluator_2
      ]
     }
     services [
      Evaluating...Select Weapon {
       class=FortAthenaBTService_AIEvaluator
       properties {
        interval=0.2
        nodeName=Evaluating...Select Weapon
        randomDeviation=0.05
        aIEvaluatorClass {defaultValue=&FortAthenaAIBotEvaluator_WeaponSelection}
       }
      }
      Evaluating...Melee Attack {
       class=FortAthenaBTService_AIEvaluator
       properties {
        interval=0.3
        nodeName=Evaluating...Melee Attack
        aIEvaluatorClass {defaultValue=&FortAthenaAIBotEvaluator_MeleeAttack}
       }
      }
      Evaluating...Range Attack {
       class=FortAthenaBTService_AIEvaluator
       properties {
        bCallTickOnSearchStart=true
        interval=0.2
        nodeName=Evaluating...Range Attack
        randomDeviation=0.05
        aIEvaluatorClass {defaultValue=&FortAthenaAIBotEvaluator_RangeAttack}
       }
      }
      Executing...Weapons Actions {class=FortAthenaBTService_ManageWeapon, properties {interval=1, randomDeviation=0.3}}
      Evaluate...Sprinting {
       class=FortAthenaBTService_AIEvaluator
       properties {interval=0.2, nodeName=Evaluate...Sprinting, aIEvaluatorClass {defaultValue=/Script/FortniteAIServer.FortAthenaAIBotEvaluator_Sprinting}}
      }
     ]
    }
    Not attacking threat {
     class=BTComposite_Selector
     children [
      LKP {
       class=BTComposite_Selector
       children [Loop LKP {
         class=BTComposite_Selector
         children [
          Selector {
           class=BTComposite_Selector
           children [
            Move To LKP {
             class=FortAthenaBTTask_BotMoveTo
             decorators [Should move to LKP {
               class=BTDecorator_Blackboard
               properties {
                arithmeticOperation=Equal
                intValue=0
                nodeName=Should move to LKP
                stringValue=""
                blackboardKey {selectedKeyName=AIEvaluator_Investigate_MoveToDestination}
               }
              }]
             properties {
              movementResultKeyName=AIEvaluator_Investigate_MovementState
              nodeName=Move To LKP
              acceptableRadius {defaultValue=50}
              allowRandomWobble {defaultValue=false}
              bAllowStrafe {defaultValue=false}
              bReachTestIncludesAgentRadius {defaultValue=false}
              bReachTestIncludesGoalRadius {defaultValue=false}
              bTrackMovingGoal {defaultValue=false}
              blackboardKey {selectedKeyName=AIEvaluator_Investigate_Destination}
              observedBlackboardValueTolerance {defaultValue=4.75}
             }
            }
            Wait {class=BTTask_Wait, properties {randomDeviation {defaultValue=0.5}, waitTime {defaultValue=0.75}}}
           ]
           decorators [Should Investigate {
             class=BTDecorator_Blackboard
             properties {
              arithmeticOperation=Equal
              intValue=4
              nodeName=Should Investigate
              stringValue=EExecutionStatus::ExecutionAllowed
              blackboardKey {selectedKeyName=AIEvaluator_Investigate_ExecutionStatus}
             }
            }]
           properties {
            children [
             {
              childComposite=null
              childTask=&self:FortAthenaBTTask_BotMoveTo_3
              decorators [&self:BTDecorator_Blackboard_29]
             }
             {childComposite=null, childTask=&self:BTTask_Wait_12, decorators []}
            ]
            services []
           }
           services []
          }
          Wait {class=BTTask_Wait, properties {randomDeviation {defaultValue=0.5}, waitTime {defaultValue=0.75}}}
         ]
         decorators [Loop {class=BTDecorator_Loop}]
         properties {
          nodeName=Loop LKP
          children [
           {
            childComposite=&self:BTComposite_Selector_12
            childTask=null
            decorators [&self:BTDecorator_Blackboard_27]
           }
           {childComposite=null, childTask=&self:BTTask_Wait_9, decorators []}
          ]
          services []
         }
         services []
        }]
       decorators [IsLKP {
         class=BTDecorator_Blackboard
         properties {
          arithmeticOperation=Equal
          intValue=2
          nodeName=IsLKP
          stringValue=EAlertLevel::LKP
          blackboardKey {selectedKeyName=AIEvaluator_Global_AlertLevel}
         }
        }]
       properties {
        nodeName=LKP
        children [{
          childComposite=&self:BTComposite_Selector_10
          childTask=null
          decorators [&self:BTDecorator_Loop_5]
         }]
        services [
         &self:FortAthenaBTService_AIEvaluator_56
         &self:FortAthenaBTService_AIEvaluator_31
         &self:FortAthenaBTService_AIEvaluator_32
         &self:FortAthenaBTService_AIEvaluator_35
         &self:FortAthenaBTService_ManageWeapon_4
         &self:FortAthenaBTService_AIEvaluator_36
        ]
       }
       services [
        Evaluating...Select Weapon {
         class=FortAthenaBTService_AIEvaluator
         properties {
          interval=0.2
          nodeName=Evaluating...Select Weapon
          randomDeviation=0.05
          aIEvaluatorClass {defaultValue=&FortAthenaAIBotEvaluator_WeaponSelection}
         }
        }
        Evaluating...Investigate Treat {
         class=FortAthenaBTService_AIEvaluator
         properties {
          bCallTickOnSearchStart=true
          interval=0.5
          nodeName=Evaluating...Investigate Treat
          aIEvaluatorClass {defaultValue=/Script/FortniteAIServer.FortAthenaAIBotEvaluator_Investigate}
         }
        }
        Evaluating...Melee Attack {
         class=FortAthenaBTService_AIEvaluator
         properties {
          interval=0.3
          nodeName=Evaluating...Melee Attack
          aIEvaluatorClass {defaultValue=&FortAthenaAIBotEvaluator_MeleeAttack}
         }
        }
        Evaluating...Range Attack {
         class=FortAthenaBTService_AIEvaluator
         properties {
          interval=0.3
          nodeName=Evaluating...Range Attack
          aIEvaluatorClass {defaultValue=&FortAthenaAIBotEvaluator_RangeAttack}
         }
        }
        Executing...Weapons Actions {class=FortAthenaBTService_ManageWeapon, properties {interval=0.2, randomDeviation=0.1}}
        Evaluate...Sprinting {
         class=FortAthenaBTService_AIEvaluator
         properties {interval=0.2, nodeName=Evaluate...Sprinting, aIEvaluatorClass {defaultValue=&FortAthenaAIBotEvaluator_Sprinting}}
        }
       ]
      }
      ALERTED {
       class=BTComposite_Selector
       children [Loop Alerted {
         class=BTComposite_Selector
         children [
          Wait {
           class=BTTask_Wait
           decorators [Should Observe {
             class=BTDecorator_Blackboard
             properties {
              arithmeticOperation=GreaterOrEqual
              intValue=3
              nodeName=Should Observe
              stringValue=EExecutionStatus::ExecutionPending
              blackboardKey {selectedKeyName=AIEvaluator_Observe_ExecutionStatus}
             }
            }]
           properties {randomDeviation {defaultValue=0.25}, waitTime {defaultValue=1}}
          }
          FallBack {class=BTTask_Wait, properties {nodeName=FallBack, waitTime {defaultValue=1}}}
         ]
         decorators [Loop {class=BTDecorator_Loop}]
         properties {
          nodeName=Loop Alerted
          children [
           {
            childComposite=null
            childTask=&self:BTTask_Wait_2
            decorators [&self:BTDecorator_Blackboard_6]
           }
           {childComposite=null, childTask=&self:BTTask_Wait_8, decorators []}
          ]
          services []
         }
         services []
        }]
       decorators [IsAlerted {
         class=BTDecorator_Blackboard
         properties {
          arithmeticOperation=Equal
          intValue=1
          nodeName=IsAlerted
          stringValue=EAlertLevel::Alerted
          blackboardKey {selectedKeyName=AIEvaluator_Global_AlertLevel}
         }
        }]
       properties {
        nodeName=ALERTED
        children [{
          childComposite=&self:BTComposite_Selector_1
          childTask=null
          decorators [&self:BTDecorator_Loop_6]
         }]
        services [
         &self:FortAthenaBTService_AIEvaluator_7
         &self:FortAthenaBTService_AIEvaluator_8
        ]
       }
       services [
        Evaluating...Observe Treat {
         class=FortAthenaBTService_AIEvaluator
         properties {
          bCallTickOnSearchStart=true
          interval=0.5
          nodeName=Evaluating...Observe Treat
          aIEvaluatorClass {defaultValue=/Script/FortniteAIServer.FortAthenaAIBotEvaluator_Observe}
         }
        }
        Evaluate...Sprinting {
         class=FortAthenaBTService_AIEvaluator
         properties {interval=0.2, nodeName=Evaluate...Sprinting, aIEvaluatorClass {defaultValue=&FortAthenaAIBotEvaluator_Sprinting}}
        }
       ]
      }
      Run Behavior {
       class=BTTask_RunBehavior
       decorators [Revive {
         class=BTDecorator_Blackboard
         properties {
          arithmeticOperation=GreaterOrEqual
          intValue=3
          nodeName=Revive
          stringValue=EExecutionStatus::ExecutionPending
          blackboardKey {selectedKeyName=AIEvaluator_Revive_ExecutionStatus}
         }
        }]
       properties {behaviorAsset=/Game/Athena/AI/Phoebe/BehaviorTrees/BT_Phoebe_Revive, nodeName=Run Behavior}
      }
      UNAWARE {
       class=BTComposite_Selector
       children [Loop Unaware {
         class=BTComposite_Selector
         children [
          PlayEmote {
           class=FortAthenaBTTask_PlayEmote
           children []
           decorators [ForcePlayEmote? {
             class=BTDecorator_Blackboard
             properties {
              arithmeticOperation=Equal
              intValue=0
              nodeName=ForcePlayEmote?
              stringValue=""
              blackboardKey {selectedKeyName=AIEvaluator_Global_LastKnownPosition}
             }
            }]
           properties {bIgnoreRestartSelf=false, nodeName=PlayEmote, playEmoteExecutionStatusKeyName=AIEvaluator_PlayEmote_ExecutionStatus, bespokeEmotes []}
           services []
          }
          "Converted Patrol " {
           class=BTComposite_Selector
           children [
            Selector {
             class=BTComposite_Selector
             children [
              Move To Converter {
               class=FortAthenaBTTask_BotMoveTo
               decorators [Should Move Towards Converter? {
                 class=BTDecorator_Blackboard
                 properties {
                  arithmeticOperation=Equal
                  intValue=4
                  nodeName=Should Move Towards Converter?
                  stringValue=""
                  blackboardKey {selectedKeyName=AIEvaluator_Converted_ShouldMoveTowardsConverter}
                 }
                }]
               properties {
                movementResultKeyName=AIEvaluator_Investigate_MovementState
                nodeName=Move To Converter
                acceptableRadius {defaultValue=50}
                allowRandomWobble {defaultValue=false}
                bAllowStrafe {defaultValue=false}
                bReachTestIncludesAgentRadius {defaultValue=false}
                bReachTestIncludesGoalRadius {defaultValue=false}
                bTrackMovingGoal {defaultValue=false}
                blackboardKey {selectedKeyName=AIEvaluator_Converted_Destination}
                observedBlackboardValueTolerance {defaultValue=4.75}
               }
              }
              Sequence {
               class=BTComposite_Sequence
               children [
                Run PatrolAround Behavior {
                 class=BTTask_RunBehavior
                 decorators [Patrol Around {
                   class=BTDecorator_Blackboard
                   properties {
                    arithmeticOperation=GreaterOrEqual
                    flowAbortMode=Self
                    intValue=2
                    nodeName=Patrol Around
                    stringValue=EExecutionStatus::ExecutionSuccess
                    blackboardKey {selectedKeyName=AIEvaluator_Patrol_ExecutionStatus}
                   }
                  }]
                 properties {behaviorAsset=/Game/Athena/AI/Phoebe/BehaviorTrees/BT_Phoebe_PatrolAround, nodeName=Run PatrolAround Behavior}
                }
                Wait {class=BTTask_Wait, properties {randomDeviation {defaultValue=1}, waitTime {defaultValue=3}}}
               ]
               decorators [Allowed To Patrol Around? {
                 class=BTDecorator_Blackboard
                 properties {
                  arithmeticOperation=Equal
                  intValue=0
                  nodeName=Allowed To Patrol Around?
                  stringValue=""
                  blackboardKey {selectedKeyName=AIEvaluator_Converted_AllowPatrolAround}
                 }
                }]
               properties {
                children [
                 {
                  childTask=&self:BTTask_RunBehavior_4
                  decorators [&self:BTDecorator_Blackboard_13]
                 }
                 {childTask=&self:BTTask_Wait_5}
                ]
                services [
                 &self:FortAthenaBTService_AIEvaluator_11
                 &self:FortAthenaBTService_AIEvaluator_26
                ]
               }
               services [
                Evaluating... Patrol Around {
                 class=FortAthenaBTService_AIEvaluator
                 properties {
                  bCallTickOnSearchStart=true
                  interval=2.5
                  nodeName=Evaluating... Patrol Around
                  randomDeviation=0.5
                  aIEvaluatorClass {defaultValue=/Script/FortniteAIServer.FortAthenaAIBotEvaluator_PatrolAround}
                 }
                }
                Evaluating... SprintingInWaterOnly {
                 class=FortAthenaBTService_AIEvaluator
                 properties {
                  interval=0.2
                  nodeName=Evaluating... SprintingInWaterOnly
                  aIEvaluatorClass {defaultValue=/Game/Athena/AI/NPCs/Base/Evaluators/Evaluator_Sprinting_InWaterOnly.Evaluator_Sprinting_InWaterOnly_C}
                 }
                }
               ]
              }
             ]
             decorators []
             properties {
              children [
               {
                childComposite=null
                childTask=&self:FortAthenaBTTask_BotMoveTo_11
                decorators [&self:BTDecorator_Blackboard_34]
               }
               {
                childComposite=&self:BTComposite_Sequence_1
                childTask=null
                decorators [&self:BTDecorator_Blackboard_8]
               }
              ]
              services []
             }
             services []
            }
            Wait {
             class=BTTask_Wait
             decorators [Can Scan Around When Waiting? {
               class=BTDecorator_Blackboard
               properties {
                arithmeticOperation=Equal
                intValue=0
                nodeName=Can Scan Around When Waiting?
                stringValue=""
                blackboardKey {selectedKeyName=AIEvaluator_Converted_AllowScanAroundWhenWaiting}
               }
              }]
             properties {waitTime {defaultValue=2}}
            }
            Wait {class=BTTask_Wait, properties {waitTime {defaultValue=2}}}
           ]
           decorators [Is Converted? {
             class=BTDecorator_CheckGameplayTagsOnActor
             properties {
              bInverseCondition=false
              flowAbortMode=None
              nodeName=Is Converted?
              tagsToMatch=Any
              actorToCheck {bNoneIsAllowedValue=false, selectedKeyName=SelfActor}
              gameplayTags {gameplayTags [{tagName=Athena.Gameplay.Status.Converted}]}
             }
            }]
           properties {
            nodeName="Converted Patrol "
            children [
             {childComposite=&self:BTComposite_Selector_9, childTask=null, decorators []}
             {
              childComposite=null
              childTask=&self:BTTask_Wait_7
              decorators [&self:BTDecorator_Blackboard_24]
             }
             {childComposite=null, childTask=&self:BTTask_Wait_3, decorators []}
            ]
            services [
             &self:FortAthenaBTService_AIEvaluator_57
             &self:FortAthenaBTService_AIEvaluator_29
            ]
           }
           services [
            Evaluating... Converted Actions {
             class=FortAthenaBTService_AIEvaluator
             properties {
              interval=0.5
              nodeName=Evaluating... Converted Actions
              aIEvaluatorClass {defaultValue=/Script/FortniteAIServer.FortAthenaAIBotEvaluator_Converted}
             }
            }
            Focus on target {
             class=FortAthenaBTService_AIEvaluator
             properties {interval=0.5, nodeName=Focus on target, aIEvaluatorClass {defaultValue=&FortAthenaAIBotEvaluator_HandleFocusing}}
            }
           ]
          }
          Conversation {
           class=FortAthenaBTTask_Conversation
           children []
           decorators [Conversation? {
             class=BTDecorator_Blackboard
             properties {
              arithmeticOperation=GreaterOrEqual
              intValue=3
              nodeName=Conversation?
              stringValue=EExecutionStatus::ExecutionPending
              blackboardKey {selectedKeyName=AIEvaluator_Conversation_ExecutionStatus}
             }
            }]
           properties {
            bIgnoreRestartSelf=false
            bResetEvaluatorStatusKeyOnFinish=true
            conversationStatusKeyName=AIEvaluator_Conversation_ExecutionStatus
            nodeName=Conversation
           }
           services []
          }
          Run DynamicBlueprint {
           class=FortAthenaBTTask_DynamicBlueprint
           children []
           decorators [Can Run LocalGameplayBehavior? {
             class=BTDecorator_Blackboard
             properties {
              arithmeticOperation=GreaterOrEqual
              intValue=3
              nodeName=Can Run LocalGameplayBehavior?
              stringValue=EExecutionStatus::ExecutionPending
              blackboardKey {selectedKeyName=AIEvaluator_DynamicBlueprint_ExecutionStatus}
             }
            }]
           properties {
            bIgnoreRestartSelf=false
            dynamicBlueprintActorKeyName=AIEvaluator_DynamicBlueprint_Actor
            dynamicBlueprintStatusKeyName=AIEvaluator_DynamicBlueprint_ExecutionStatus
            nodeName=Run DynamicBlueprint
           }
           services []
          }
          Can Patrolling? {
           class=BTComposite_Selector
           children [
            Selector {
             class=BTComposite_Selector
             children [
              Run Unstuck Behavior {
               class=BTTask_RunBehavior
               decorators []
               properties {behaviorAsset=&BT_Phoebe_UnstuckNew, nodeName=Run Unstuck Behavior}
              }
              MoveTo PatrolPoint {
               class=FortAthenaBTTask_BotMoveTo
               decorators [Is Stressed? {
                 class=BTDecorator_Blackboard
                 properties {
                  arithmeticOperation=GreaterOrEqual
                  intValue=3
                  nodeName=Is Stressed?
                  stringValue=""
                  blackboardKey {selectedKeyName="AIEvaluator_IsStressed "}
                 }
                }]
               properties {
                movementResultKeyName=AIEvaluator_Patrolling_MovementState
                nodeName=MoveTo PatrolPoint
                acceptableRadius {defaultValue=10}
                allowRandomWobble {defaultValue=false}
                bAllowStrafe {defaultValue=true}
                bTrackMovingGoal {defaultValue=false}
                blackboardKey {selectedKeyName=AIEvaluator_Patrolling_Destination}
                observedBlackboardValueTolerance {defaultValue=50}
               }
              }
              MoveTo PatrolPoint {
               class=FortAthenaBTTask_BotMoveTo
               decorators []
               properties {
                movementResultKeyName=AIEvaluator_Patrolling_MovementState
                nodeName=MoveTo PatrolPoint
                acceptableRadius {defaultValue=10}
                allowRandomWobble {defaultValue=false}
                bAllowStrafe {defaultValue=true}
                bTrackMovingGoal {defaultValue=false}
                blackboardKey {selectedKeyName=AIEvaluator_Patrolling_Destination}
                observedBlackboardValueTolerance {defaultValue=50}
               }
              }
             ]
             decorators [Should move to patrol point? {
               class=BTDecorator_Blackboard
               properties {
                arithmeticOperation=Equal
                intValue=0
                nodeName=Should move to patrol point?
                stringValue=""
                blackboardKey {selectedKeyName=AIEvaluator_Patrolling_ShouldMove}
               }
              }]
             properties {
              children [
               {childComposite=null, childTask=&self:BTTask_RunBehavior_2, decorators []}
               {
                childComposite=null
                childTask=&self:FortAthenaBTTask_BotMoveTo_10
                decorators [&self:BTDecorator_Blackboard_26]
               }
               {childComposite=null, childTask=&self:FortAthenaBTTask_BotMoveTo_9, decorators []}
              ]
              services []
             }
             services []
            }
            Wait On PatrolPoint {class=BTTask_Wait, properties {nodeName=Wait On PatrolPoint, waitTime {defaultValue=5}}}
           ]
           decorators [Can Patrolling? {
             class=BTDecorator_Blackboard
             properties {
              arithmeticOperation=GreaterOrEqual
              intValue=3
              nodeName=Can Patrolling?
              stringValue=EExecutionStatus::ExecutionPending
              blackboardKey {selectedKeyName=AIEvaluator_Patrolling_ExecutionStatus}
             }
            }]
           properties {
            nodeName=Can Patrolling?
            children [
             {
              childComposite=&self:BTComposite_Selector_13
              childTask=null
              decorators [&self:BTDecorator_Blackboard_30]
             }
             {childComposite=null, childTask=&self:BTTask_Wait_14, decorators []}
            ]
            services []
           }
           services []
          }
          Fallback Behavior {
           class=BTComposite_Selector
           children [
            Run PatrolAround Behavior {
             class=BTTask_RunBehavior
             decorators [Patrol Around {
               class=BTDecorator_Blackboard
               properties {
                arithmeticOperation=GreaterOrEqual
                intValue=2
                nodeName=Patrol Around
                stringValue=EExecutionStatus::ExecutionSuccess
                blackboardKey {selectedKeyName=AIEvaluator_Patrol_ExecutionStatus}
               }
              }]
             properties {behaviorAsset=&BT_Phoebe_PatrolAround, nodeName=Run PatrolAround Behavior}
            }
            Wait {class=BTTask_Wait, properties {waitTime {defaultValue=2}}}
           ]
           decorators [Can Patrol Around? {
             class=BTDecorator_Blackboard
             properties {
              arithmeticOperation=Equal
              intValue=0
              nodeName=Can Patrol Around?
              stringValue=""
              blackboardKey {selectedKeyName=AIEvaluator_CanUseFallbackPatrolAround}
             }
            }]
           properties {
            nodeName=Fallback Behavior
            children [
             {
              childComposite=null
              childTask=&self:BTTask_RunBehavior_3
              decorators [&self:BTDecorator_Blackboard_32]
             }
             {childComposite=null, childTask=&self:BTTask_Wait_15, decorators []}
            ]
            services [
             &self:FortAthenaBTService_AIEvaluator_17
             &self:FortAthenaBTService_AIEvaluator_37
            ]
           }
           services [
            Evaluating... Patrol Around {
             class=FortAthenaBTService_AIEvaluator
             properties {
              bCallTickOnSearchStart=true
              interval=2.5
              nodeName=Evaluating... Patrol Around
              randomDeviation=0.5
              aIEvaluatorClass {defaultValue=&FortAthenaAIBotEvaluator_PatrolAround}
             }
            }
            Focus on target {
             class=FortAthenaBTService_AIEvaluator
             properties {interval=0.5, nodeName=Focus on target, aIEvaluatorClass {defaultValue=&FortAthenaAIBotEvaluator_HandleFocusing}}
            }
           ]
          }
          Wait {class=BTTask_Wait, properties {waitTime {defaultValue=2}}}
         ]
         decorators [Loop {class=BTDecorator_Loop}]
         properties {
          nodeName=Loop Unaware
          children [
           {
            childComposite=null
            childTask=&self:FortAthenaBTTask_PlayEmote_3
            decorators [&self:BTDecorator_Blackboard_18]
           }
           {
            childComposite=&self:BTComposite_Selector_0
            childTask=null
            decorators [&self:BTDecorator_CheckGameplayTagsOnActor_1]
           }
           {
            childComposite=null
            childTask=&self:FortAthenaBTTask_Conversation_0
            decorators [&self:BTDecorator_Blackboard_7]
           }
           {
            childComposite=null
            childTask=&self:FortAthenaBTTask_DynamicBlueprint_1
            decorators [&self:BTDecorator_Blackboard_11]
           }
           {
            childComposite=&self:BTComposite_Selector_18
            childTask=null
            decorators [&self:BTDecorator_Blackboard_33]
           }
           {
            childComposite=&self:BTComposite_Selector_17
            childTask=null
            decorators [&self:BTDecorator_Blackboard_31]
           }
           {childComposite=null, childTask=&self:BTTask_Wait_16, decorators []}
          ]
          services []
         }
         services []
        }]
       decorators []
       properties {
        nodeName=UNAWARE
        children [{
          childComposite=&self:BTComposite_Selector_25
          childTask=null
          decorators [&self:BTDecorator_Loop_9]
         }]
        services [
         &self:FortAthenaBTService_AIEvaluator_61
         &self:FortAthenaBTService_AIEvaluator_46
         &self:FortAthenaBTService_AIEvaluator_47
         &self:FortAthenaBTService_AIEvaluator_48
         &self:FortAthenaBTService_ManageWeapon_9
         &self:FortAthenaBTService_AIEvaluator_50
        ]
       }
       services [
        Evaluating...Select Weapon {
         class=FortAthenaBTService_AIEvaluator
         properties {
          interval=0.2
          nodeName=Evaluating...Select Weapon
          randomDeviation=0.05
          aIEvaluatorClass {defaultValue=&FortAthenaAIBotEvaluator_WeaponSelection}
         }
        }
        Evaluating...Patrolling {
         class=FortAthenaBTService_AIEvaluator
         properties {interval=0.5, nodeName=Evaluating...Patrolling, aIEvaluatorClass {defaultValue=/Script/FortniteAIServer.FortAthenaNpcEvaluator_Patrolling}}
        }
        Evaluating...Melee Attack {
         class=FortAthenaBTService_AIEvaluator
         properties {
          interval=0.3
          nodeName=Evaluating...Melee Attack
          aIEvaluatorClass {defaultValue=&FortAthenaAIBotEvaluator_MeleeAttack}
         }
        }
        Evaluating...Range Attack {
         class=FortAthenaBTService_AIEvaluator
         properties {
          interval=1
          nodeName=Evaluating...Range Attack
          aIEvaluatorClass {defaultValue=&FortAthenaAIBotEvaluator_RangeAttack}
         }
        }
        Executing...Weapons Actions {class=FortAthenaBTService_ManageWeapon, properties {interval=0.2, randomDeviation=0.1}}
        Evaluating...StressTag {
         class=FortAthenaBTService_AIEvaluator
         properties {
          interval=0.5
          nodeName=Evaluating...StressTag
          aIEvaluatorClass {defaultValue=/Game/Athena/AI/NPCs/Base/Evaluators/Evaluator_StressTag_Guard.Evaluator_StressTag_Guard_C}
         }
        }
       ]
      }
      Fallback {class=BTTask_Wait, properties {nodeName=Fallback, waitTime {defaultValue=5}}}
     ]
     decorators []
     properties {
      nodeName=Not attacking threat
      children [
       {
        childComposite=&self:BTComposite_Selector_8
        childTask=null
        decorators [&self:BTDecorator_Blackboard_9]
       }
       {
        childComposite=&self:BTComposite_Selector_3
        childTask=null
        decorators [&self:BTDecorator_Blackboard_1]
       }
       {
        childComposite=null
        childTask=&self:BTTask_RunBehavior_7
        decorators [&self:BTDecorator_Blackboard_17]
       }
       {childComposite=&self:BTComposite_Selector_26, childTask=null, decorators []}
       {childComposite=null, childTask=&self:BTTask_Wait_6, decorators []}
      ]
      services [&self:FortAthenaBTService_AIEvaluator_16]
     }
     services [Evaluating...Revive {
       class=FortAthenaBTService_AIEvaluator
       properties {
        interval=1
        nodeName=Evaluating...Revive
        aIEvaluatorClass {defaultValue=/Game/Athena/AI/NPCs/Base/Evaluators/Evaluator_Revive_Guard.Evaluator_Revive_Guard_C}
       }
      }]
    }
   ]
   decorators []
   properties {
    nodeName=On Ground
    children [
     {
      childComposite=null
      childTask=&self:BTTask_Wait_10
      decorators [
       &self:BTDecorator_Blackboard_12
       &self:BTDecorator_Loop_1
      ]
     }
     {
      childComposite=null
      childTask=&self:BTTask_Wait_0
      decorators [
       &self:BTDecorator_Blackboard_5
       &self:BTDecorator_Loop_2
      ]
     }
     {
      childComposite=null
      childTask=&self:FortAthenaBTTask_SteerMovement_0
      decorators [&self:BTDecorator_Blackboard_21]
     }
     {
      childComposite=null
      childTask=&self:FortAthenaBTTask_ShootTrap_0
      decorators [&self:BTDecorator_Blackboard_0]
     }
     {
      childComposite=null
      childTask=&self:FortAthenaBTTask_BotMoveTo_1
      decorators [&self:BTDecorator_Blackboard_3]
     }
     {
      childComposite=null
      childTask=&self:FortAthenaBTTask_BotMoveTo_7
      decorators [&self:BTDecorator_Blackboard_16]
     }
     {
      childComposite=&self:BTComposite_Selector_6
      childTask=null
      decorators [
       &self:BTDecorator_Blackboard_2
       &self:BTDecorator_Loop_4
      ]
     }
     {
      childComposite=null
      childTask=&self:FortAthenaBTTask_Zipline_0
      decorators [&self:BTDecorator_Blackboard_50]
     }
     {
      childComposite=&self:BTComposite_Selector_16
      childTask=null
      decorators [&self:BTDecorator_Blackboard_14]
     }
     {
      childComposite=&self:BTComposite_Selector_4
      childTask=null
      decorators [&self:BTDecorator_Blackboard_4]
     }
     {childComposite=&self:BTComposite_Selector_15, childTask=null, decorators []}
    ]
    services [
     &self:FortAthenaBTService_AIEvaluator_3
     &self:FortAthenaBTService_UpdateTarget_0
     &self:FortAthenaBTService_AIEvaluator_43
     &self:FortAthenaBTService_AIEvaluator_14
     &self:FortAthenaBTService_AIEvaluator_66
     &self:FortAthenaBTService_Crouch_0
     &self:FortAthenaBTService_Sprinting_1
     &self:FortAthenaBTService_Jump_1
     &self:FortAthenaBTService_AIEvaluator_49
     &self:FortAthenaBTService_AIEvaluator_58
     &self:FortAthenaBTService_AIEvaluator_6
     &self:FortAthenaBTService_JetpackStrafe_0
    ]
   }
   services [
    Evaluating... Leash {
     class=FortAthenaBTService_AIEvaluator
     properties {interval=0.5, nodeName=Evaluating... Leash, aIEvaluatorClass {defaultValue=/Script/FortniteAIServer.FortAthenaNpcEvaluator_Leash}}
    }
    Evaluating...Targets {class=FortAthenaBTService_UpdateTarget, properties {interval=0.3}}
    Evaluating...AvoidThreat {
     class=FortAthenaBTService_AIEvaluator
     properties {
      interval=0.5
      nodeName=Evaluating...AvoidThreat
      aIEvaluatorClass {defaultValue=/Game/Athena/AI/Phoebe/Evaluators/Phoebe_Evaluator_AvoidThreat.Phoebe_Evaluator_AvoidThreat_C}
     }
    }
    Evaluating.. Propagate Awareness {
     class=FortAthenaBTService_AIEvaluator
     properties {
      interval=1
      nodeName=Evaluating.. Propagate Awareness
      aIEvaluatorClass {defaultValue=/Game/Athena/AI/NPCs/Base/Evaluators/Evaluator_PropagateAwareness_Guard.Evaluator_PropagateAwareness_Guard_C}
     }
    }
    Evaluating...Character Launched {
     class=FortAthenaBTService_AIEvaluator
     properties {
      interval=0.5
      nodeName=Evaluating...Character Launched
      aIEvaluatorClass {defaultValue=/Script/FortniteAIServer.FortAthenaAIBotEvaluator_CharacterLaunched}
     }
    }
    Executing...Crouch {class=FortAthenaBTService_Crouch}
    Executing...Sprinting {class=FortAthenaBTService_Sprinting}
    Executing...Jump {class=FortAthenaBTService_Jump, properties {crouchExecutionStatusName=AIEvaluator_Crouch_ExecutionStatus}}
    Evaluating...Danger {
     class=FortAthenaBTService_AIEvaluator
     properties {
      interval=1000000000
      nodeName=Evaluating...Danger
      randomDeviation=0
      aIEvaluatorClass {defaultValue=/Game/Athena/AI/Phoebe/Evaluators/Phoebe_Evaluator_DangerDetection.Phoebe_Evaluator_DangerDetection_C}
     }
    }
    Evaluating...Trap On Path {
     class=FortAthenaBTService_AIEvaluator
     properties {
      interval=0.3
      nodeName=Evaluating...Trap On Path
      aIEvaluatorClass {defaultValue=/Script/FortniteAIServer.FortAthenaAIBotEvaluator_TrapOnPathDetected}
     }
    }
    Evaluate...Sand Tunnel {
     class=FortAthenaBTService_AIEvaluator
     properties {
      interval=100000000
      nodeName=Evaluate...Sand Tunnel
      aIEvaluatorClass {defaultValue=/Game/Athena/AI/Phoebe/Evaluators/Phoebe_Evaluator_SandTunnel.Phoebe_Evaluator_SandTunnel_C}
     }
    }
    Executing...Jetpack Strafe {
     class=FortAthenaBTService_JetpackStrafe
     properties {
      bCallTickOnSearchStart=false
      bRestartTimerOnEachActivation=false
      crouchExecutionStatusName=AIEvaluator_Crouch_ExecutionStatus
      interval=0.5
      jetpackStrafeExecutionStatusName=AIEvaluator_JetpackStrafe_ExecutionStatus
      jumpExecutionStatusName=AIEvaluator_Jump_ExecutionStatus
      jumpInputReleaseDelay=1
      nodeName=Executing...Jetpack Strafe
      randomDeviation=0.1
     }
    }
   ]
  }
 ]
 decorators []
 properties {
  nodeName=Alive
  children [
   {
    childComposite=null
    childTask=&self:FortAthenaBTTask_Dive_0
    decorators [&self:BTDecorator_Blackboard_15]
   }
   {
    childComposite=null
    childTask=&self:FortAthenaBTTask_Glide_0
    decorators [&self:BTDecorator_Blackboard_22]
   }
   {childComposite=&self:BTComposite_Selector_2, childTask=null, decorators []}
  ]
  services [&self:FortAthenaBTService_AIEvaluator_1]
 }
 services [Evaluating...Free Fall {
   class=FortAthenaBTService_AIEvaluator
   properties {
    bCallTickOnSearchStart=true
    interval=0.5
    nodeName=Evaluating...Free Fall
    aIEvaluatorClass {defaultValue=/Game/Athena/AI/Phoebe/Evaluators/Phoebe_Evaluator_FreeFall.Phoebe_Evaluator_FreeFall_C}
   }
  }]
}
